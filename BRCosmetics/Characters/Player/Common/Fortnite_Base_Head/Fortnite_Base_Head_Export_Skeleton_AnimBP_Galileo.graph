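# Facial animation Blueprint for the Galileo head on the shared Fortnite base head skeleton.
# It copies the parent mesh's pose, layers a default face idle from faceAttach, and runs the result through the Galileo pose asset.
# Facial anim type selects whether a canned dialogue pose overlays jaw, cheek and lip bones by voice weight.
# The LiveLinkFacial cache only repeats the game pose, and the voice weight compares constant zero against its threshold.
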AnimBlueprint /BRCosmetics/Characters/Player/Common/Fortnite_Base_Head/Fortnite_Base_Head_Export_Skeleton_AnimBP_Galileo
target_skeleton=/Game/Characters/Player/Common/Fortnite_Base_Head/Fortnite_Base_Head_Export_Skeleton
graphs [
 AnimGraph {nodes [
   n1 = Layered blend per bone(
    BasePose=Use cached pose 'PostPoseAsset'()
    BlendPoses_0=Use cached pose 'AmplitudeFacial'()
    BlendWeights_0=Get VoiceAmplitudeBlendWeight() -> (VoiceAmplitudeBlendWeight:double)
   ) {layerSetup [{branchFilters [
       {boneName=C_jaw}
       {boneName=L_cheeck_inner}
       {boneName=R_cheek_inner}
       {boneName=C_lip_upper_mid}
       {boneName=L_lip_upper_outer}
       {boneName=R_lip_upper_outer}
       {boneName=teeth_upper}
       {boneName=C_nose_bridge}
      ]}]}
   n2 = "Blend Poses (EFortFacialAnimTypes)"!(
    BlendPose_0:Pose Link Structure=Use cached pose 'PostPoseAsset'()
    BlendPose_1:Pose Link Structure=@n1
    BlendTime_0:float=0.1
    BlendTime_1:float=0.1
    ActiveEnumValue:Byte=Get!() -> (
     FacialInput_SubjectName:Live Link Subject Name Structure
     FacialInput_CurrentAnimType:EFortFacialAnimTypes Enum
     FacialInput_bCurvesOnly:bool=~
    ).FacialInput_CurrentAnimType
   ) -> (Pose:Pose Link Structure) {
    is=title AnimGraphNode_BlendListByEnum
    blendType=Linear
    blendPose [{linkId=-1, sourceLinkId=-1}, {linkId=-1, sourceLinkId=-1}]
    blendTime [0.1,0.1]
   }
   Output Pose(Result=@n2) {layerGroup=None}
   Save cached pose 'GameFacial'(Pose=Layered blend per bone(
     BasePose=Copy Pose From Mesh() {bCopyCurves=true, bUseAttachedParent=true}
     BlendPoses_0=Sequence Player 'DefaultFaceIdle_Legacy'() {sequence=/Game/Animation/Game/MainPlayer/Menu/FACIAL/DefaultFaceIdle_Legacy}
    ) {layerSetup [{branchFilters [{boneName=faceAttach}]}]}) {cacheName=GameFacial}
   n3 = "Blend Poses (EFortFacialAnimTypes)"!(
    BlendPose_0=Use cached pose 'GameFacial'()
    BlendPose_1=Layered blend per bone(BasePose=Use cached pose 'GameFacial'(), BlendPoses_0=Use cached pose 'LiveLinkFacial'()) {layerSetup [{branchFilters [{boneName=faceAttach}]}]}
    BlendPose_2:Pose Link Structure=Use cached pose 'LiveLinkFacial'()
    BlendTime_0=0.1
    BlendTime_1=0.1
    BlendTime_2:float=0.1
    ActiveEnumValue=Get() -> (FacialInput_bCurvesOnly=~).FacialInput_CurrentAnimType
   ) -> (Pose) {
    blendType=Linear
    blendPose [{linkId=-1, sourceLinkId=-1}, {linkId=-1, sourceLinkId=-1}, {linkId=-1, sourceLinkId=-1}]
    blendTime [0.1,0.1,0.1]
   }
   n4 = Male_Medium_Galileo_Jumper_Head_Facial_Poses_PoseAsset!(SourcePose:Pose Link Structure=@n3) -> (Pose:Pose Link Structure) {
    is=title AnimGraphNode_PoseBlendNode
    blendOption=Linear
    poseAsset=/BRCosmetics/Characters/Player/Male/Medium/Heads/M_MED_Galileo_Jumper_Head/Meshes/Male_Medium_Galileo_Jumper_Head_Facial_Poses_PoseAsset
   }
   Save cached pose 'PostPoseAsset'(Pose=@n4) {cacheName=PostPoseAsset}
   n5 = Male_Medium_Galileo_Jumper_Head_Facial_Poses_PoseAsset(SourcePose=Sequence Player 'Galileo_dialogue_01'() {sequence=/BRCosmetics/Characters/Player/Common/Fortnite_Base_Head/Anim_Tests/Galileo_dialogue_01}) {
    blendOption=Linear
    poseAsset=&Male_Medium_Galileo_Jumper_Head_Facial_Poses_PoseAsset
   }
   Save cached pose 'AmplitudeFacial'(Pose=@n5) {cacheName=AmplitudeFacial}
   n6 = Break Fort Anim Input Facial!(FortAnimInput_Facial:"Fort Anim Input Facial Structure (by ref)"=Get FacialInput() -> (FacialInput:Fort Anim Input Facial Structure)) -> (SubjectName:Live Link Subject Name Structure, CurrentAnimType:EFortFacialAnimTypes Enum, bCurvesOnly:bool=~) {is=title K2Node_BreakStruct}
   Save cached pose 'LiveLinkFacial'(Pose=Blend Poses by bool(BlendPose_0=Use cached pose 'GameFacial'(), BlendPose_1=Use cached pose 'GameFacial'(), bActiveValue=@n6.bCurvesOnly) {blendType=Linear}) {cacheName=LiveLinkFacial}
   Reroute Node(InputPin:Live Link Subject Name Structure=@n6) -> (OutputPin:Live Link Subject Name Structure)
  ]}
 EventGraph {nodes [
   Event BlueprintUpdateAnimation()
   > Cast To FortPlayerPawnAthena(Object=TryGetPawnOwner())
   n1 = Get VoiceThreshold() -> (VoiceThreshold:double)
   > Set VoiceAmplitudeBlendWeight(VoiceAmplitudeBlendWeight:double="To Float (Boolean)"(InBool="float >= float"(A=0, B=@n1) {is=title K2Node_CallFunction})) -> (Output_Get:double)
   Event BlueprintInitializeAnimation()
   > Set MeshToCopy(MeshToCopy:Skeletal Mesh Component Object Reference=GetOwnerSkeletalMeshComponent()) -> (Output_Get:Skeletal Mesh Component Object Reference)
   > Set GetPlayerListUsesCustomCharacterParts(GetPlayerListUsesCustomCharacterParts:bool=GetPlaylistUsesCustomCharacterParts!() -> (ReturnValue:bool)) -> (Output_Get:bool)
  ]}
]
variables [
 GetPlayerListUsesCustomCharacterParts {type=boolean}
 MeshToCopy {type=SkeletalMeshComponent}
 VoiceAmplitudeBlendWeight {type=number}
 VoiceThreshold {type=number, default=0.01}
]
